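# Translucent fog card material for the KelpLinen environment, with optional camera-facing world position offset.
# Two scrolling noise layers, distorted by micro noise and offset by actor position, shape opacity alongside edge bands, macro noise, depth fade and optional location fade.
# A Phase_Tracker collection parameter switches camera-height fade levels and time-of-day colors.
# The last phase adds a sphere-masked glow around a source point, and audio beats modulate its radius and emissive strength.
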Material /Clyde/Environments/KelpLinen/Materials/Fog/M_KelpLinen_FogCard
graph {
 properties {bAutomaticallySetUsageInEditor=false, blendMode=BLEND_Translucent, usedShadingModels=Default Lit}
 nodes [
  n1 = Add(A=TextureCoordinate(), B=NamedRerouteUsage())
  n2 = VectorParameter() {parameterName=Macro.OffsetTiling, channelNames {a=Tiling Y, b=Tiling X, g=Offset Y, r=Offset X}, defaultValue=<rgba 0 0 0.2 0.2>}
  n3 = AppendVector(A=@n2.Offset X, B=@n2.Offset X)
  n4 = MaterialFunctionCall() {materialFunction=&ObjectScale}
  n5 = Divide(A=@n4.Scale X, B=@n4.Scale X)
  n6 = Constant() {r=1}
  n7 = Multiply(A=Multiply(A=Add(A=@n1, B=@n3), B=AppendVector(A=@n5, B=@n6)), B=AppendVector(A=@n2.Tiling X, B=@n2.Tiling X)) {constB=0.2}
  Add(A=@n7, B=NamedRerouteUsage())
  n8 = MaterialFunctionCall() {materialFunction=&ObjectScale}
  n9 = Divide(A=@n8.Scale X, B=@n8.Scale X)
  n10 = Constant() {r=1}
  n11 = Multiply(A=AppendVector(A=@n9, B=@n10), B=ScalarParameter() {defaultValue=1, parameterName=Fog.Tiling.A})
  n12 = MaterialFunctionCall(Float2=TextureCoordinate()) {materialFunction=&BreakOutFloat2Components}
  n13 = ScalarParameter() {defaultValue=0.02, parameterName=Fog.Micro.Speed}
  n14 = Time()
  n15 = ComponentMask(None=ActorPositionWS()) {r=true}
  n16 = Add(A=Multiply(A=Multiply(A=@n13) {constB=-1}, B=Add(A=@n14, B=@n15)), B=@n12.G)
  n17 = MaterialFunctionCall() {materialFunction=&ObjectScale}
  n18 = AppendVector(A=Divide(A=@n17.Scale X, B=@n17.Scale X), B=Constant() {r=1})
  n19 = ScalarParameter() {defaultValue=1.7, parameterName=Fog.Micro.Tiling}
  n20 = TextureSampleParameter2D(UVs=Multiply(
    A=MaterialFunctionCall(X=@n12.R, Y=@n16) {materialFunction=&MakeFloat2}.Result
    B=Multiply(A=@n18, B=@n19)
   )) {parameterName=MicoNoise, texture=&/T_KelpLinen_CloudNoise_low}
  n21 = AppendVector(A=@n20.R, B=@n20.R)
  n22 = Multiply(A=Subtract(A=@n21) {constB=0.5}, B=ScalarParameter() {defaultValue=0.05, parameterName=Fog.Micro.Blend})
  n23 = Add(A=TextureCoordinate(), B=@n22)
  n24 = VectorParameter() {parameterName=Fog.Noise.Speed, channelNames {a=Multiplier, b=Z, g=Y, r=X}, defaultValue=<rgba 0.5 0.5 -1 100>}
  n25 = Add(A=Time(), B=ComponentMask(None=ActorPositionWS()) {r=true})
  n26 = Multiply(A=@n25, B=ComponentMask(None=Multiply(A=@n24, B=@n24)) {g=true, r=true})
  n27 = TextureObjectParameter() {
   parameterName=FogNoise
   samplerType=SAMPLERTYPE_LinearColor
   texture=&/T_KelpLinen_CloudNoise_low_2
  }
  n28 = TextureSample(UVs=Add(A=Multiply(A=@n23, B=@n11), B=Multiply(A=@n26) {constB=0.005}), Tex=@n27)
  n29 = Multiply(A=@n23, B=Multiply(A=AppendVector(A=@n9, B=@n10), B=ScalarParameter() {defaultValue=1, parameterName=Fog.Tiling.B}))
  n30 = VectorParameter() {parameterName=Fog.Noise.Speed, channelNames {a=Multiplier, b=Z, g=Y, r=X}, defaultValue=<rgba 0.5 0.5 -1 100>}
  n31 = Multiply(A=@n30, B=@n30)
  n32 = Constant3Vector() {constant=<rgba -0.66 0.66 0.66 0>}
  n33 = Multiply(A=@n25, B=ComponentMask(None=Multiply(A=@n31, B=@n32)) {g=true, r=true})
  n34 = TextureSample(UVs=Add(A=Multiply(A=@n29) {constB=0.8}, B=Multiply(A=@n33) {constB=0.005}), Tex=@n27)
  AppendVector(A=@n28.R, B=@n34.G)
  n35 = MaterialFunctionCall() {materialFunction=&ObjectScale}
  MaterialFunctionCall!(
   Center Location
   0-1 UVs
   XY scale=Multiply(A=AppendVector(A=@n35.Scale X, B=@n35.Scale X)) {constB=100}
   OverridePivot
   "Normals (Optional)"
  ) {materialFunction=&Sprite}
  Multiply(B=ScalarParameter() {defaultValue=1, parameterName=Scale})
  n36 = TextureCoordinate()
  n37 = ScalarParameter() {defaultValue=0.2, parameterName=Fog.UVOffsetMulti}
  n38 = Power(Base=@n28.B) {constExponent=8}
  n39 = Multiply(A=Saturate(None=Multiply(A=@n38) {constB=8})) {constB=0.3}
  n40 = Saturate(None=Multiply(A=Power(Base=@n28.G) {constExponent=4}) {constB=8})
  n41 = Saturate(None=Multiply(A=Power(Base=@n28.R) {constExponent=8}) {constB=8})
  n42 = LinearInterpolate(A=Multiply(A=@n40) {constB=0.66}, Alpha=@n41)
  n43 = Max(A=@n41, B=@n40)
  n44 = Power(Base=@n34.B) {constExponent=8}
  n45 = Multiply(A=Saturate(None=Multiply(A=@n44) {constB=8})) {constB=0.3}
  n46 = Saturate(None=Multiply(A=Power(Base=@n34.G) {constExponent=4}) {constB=8})
  n47 = Saturate(None=Multiply(A=Power(Base=@n34.R) {constExponent=8}) {constB=8})
  n48 = LinearInterpolate(A=Multiply(A=@n46) {constB=0.66}, Alpha=@n47)
  n49 = Max(A=@n47, B=@n46)
  n50 = Power(
   Base=AppendVector(A=LinearInterpolate(A=@n39, B=@n42, Alpha=@n43), B=LinearInterpolate(A=@n45, B=@n48, Alpha=@n49))
   Exp=ScalarParameter() {defaultValue=1, parameterName=Opacity.Contrast}
  )
  n51 = Add(A=@n36, B=Subtract(A=Multiply(A=@n37, B=@n50)) {constB=0.1})
  n52 = Power(Base=ComponentMask(None=@n51) {g=true})
  n53 = ComponentMask(None=OneMinus(None=@n36)) {g=true}
  n54 = Multiply(A=Saturate(None=Multiply(A=@n53) {constB=5}), B=@n52)
  n55 = ScalarParameter() {defaultValue=1, parameterName=Opacity.BottomSubtractAlpha}
  n56 = Constant() {r=0.49}
  n57 = Constant() {r=0.01}
  n58 = StaticBool() {value=true}
  n59 = Multiply(
   A=Multiply(
    A=LinearInterpolate(A=@n52, B=@n54, Alpha=@n55)
    B=MaterialFunctionCall!(Width=@n56, Sharpness=@n57, Offset, Direction Switch=@n58, Compare, Input Coordinates=@n51) {materialFunction=/Engine/Functions/Engine_MaterialFunctions03/Procedurals/GeneratedBand}.Result
   )
   B=MaterialFunctionCall(Width=@n56, Sharpness=@n57, Direction Switch=StaticBool() {value=true}) {materialFunction=&GeneratedBand}.Result
  )
  n60 = LinearInterpolate(A=Multiply(A=@n59, B=NamedRerouteUsage()), B=@n59, Alpha=ScalarParameter() {defaultValue=0.6, parameterName=Opacity.Density})
  n61 = NamedRerouteUsage()
  n62 = Power(
   Base=TextureSample(UVs=@n7) {samplerType=SAMPLERTYPE_LinearColor, texture=&T_KelpLinen_CloudNoise_low_2}.R
   Exp=ScalarParameter() {defaultValue=1, parameterName=Macro.Contrast}
  )
  n63 = ScalarParameter() {defaultValue=1, parameterName=Macro.Multi}
  n64 = Multiply(
   A=LinearInterpolate(A=@n60, B=Multiply(A=@n60, B=@n61), Alpha=ScalarParameter() {parameterName=Fog.Micro.OpacityBlend})
   B=Saturate(None=Multiply(A=@n62, B=@n63))
  )
  n65 = LinearInterpolate(
   A=@n64
   B=Multiply(A=@n64, B=MaterialFunctionCall() {materialFunction=/ClydeCosmetics/Environments/KelpLinen/MF_KelpLinen_LocationFade}.Output)
   Alpha=ScalarParameter() {parameterName=Opacity.UseLocationFade}
  )
  n66 = ScalarParameter() {defaultValue=1000, parameterName=Opacity.DepthFade}
  Multiply(
   A=NamedRerouteUsage()
   B=Multiply(A=DepthFade(Opacity=@n65, FadeDistance=@n66), B=ScalarParameter() {defaultValue=1, parameterName=Opacity.Multi})
  )
  NamedRerouteDeclaration(None=Multiply(A=@n28.R, B=@n34.G)) {name=Noise, nodeColor=<rgba 1 0 0.0234 1>}
  NamedRerouteDeclaration(None=Power(Base=@n28.B, Exp=ScalarParameter() {defaultValue=3, parameterName=Color.SecondaryNoise.Contrast}) {constExponent=3}) {name=Noise_Alt, nodeColor=<rgba 1 0 0.0234 1>}
  NamedRerouteDeclaration(None=@n27) {name=FogNoiseTexture, nodeColor=<rgba 0.1016 1 0 1>}
  NamedRerouteDeclaration(None=Desaturation(None=@n21)) {name=MicroNoise, nodeColor=<rgba 0 1 0.7188 1>}
  NamedRerouteDeclaration(None=@n15) {name=ActorRandom, nodeColor=<rgba 1 0 0.9844 1>}
  NamedRerouteDeclaration(None=@n22) {name=FogMicroUVs, nodeColor=<rgba 0 1 0.1094 1>}
  MaterialExpressionScalarParameter_21 = ScalarParameter() {defaultValue=0.3, parameterName=HeightFade.MinFade.Phase1}
  MaterialExpressionScalarParameter_22 = ScalarParameter() {defaultValue=0.3, parameterName=HeightFade.MinFade.Phase2}
  MaterialExpressionScalarParameter_26 = ScalarParameter() {defaultValue=0.3, parameterName=HeightFade.MinFade.Phase3}
  n67 = Switch!(
   SwitchValue=CollectionParameter() {collection=/ClydeCosmetics/Environments/KelpLinen/Materials/MPC_KelpLinen, parameterName=Phase_Tracker}
   Default=@MaterialExpressionScalarParameter_21
   0=@MaterialExpressionScalarParameter_21
   1=@MaterialExpressionScalarParameter_21
   2=@MaterialExpressionScalarParameter_22
   3=@MaterialExpressionScalarParameter_26
  ) {inputs [
    {
     inputName=0
     input {
      expression=&self:MaterialExpressionScalarParameter_21
      inputName=None
      mask=0
      maskA=0
      maskB=0
      maskG=0
      maskR=0
      outputIndex=0
     }
    }
    {
     inputName=1
     input {
      expression=&self:MaterialExpressionScalarParameter_21
      inputName=None
      mask=0
      maskA=0
      maskB=0
      maskG=0
      maskR=0
      outputIndex=0
     }
    }
    {
     inputName=2
     input {
      expression=&self:MaterialExpressionScalarParameter_22
      inputName=None
      mask=0
      maskA=0
      maskB=0
      maskG=0
      maskR=0
      outputIndex=0
     }
    }
    {
     inputName=3
     input {
      expression=&self:MaterialExpressionScalarParameter_26
      inputName=None
      mask=0
      maskA=0
      maskB=0
      maskG=0
      maskR=0
      outputIndex=0
     }
    }
   ]}
  n68 = Subtract(A=ComponentMask(None=CameraPositionWS()) {b=true}, B=ScalarParameter() {defaultValue=10000, parameterName=HeightFade.Height})
  n69 = Saturate(None=Divide(A=Max(A=@n68) {constB=0.001}, B=ScalarParameter() {defaultValue=5000, parameterName=HeightFade.Distance}))
  NamedRerouteDeclaration(None=LinearInterpolate(A=@n67, Alpha=OneMinus(None=@n69))) {name=HeightOpacity, nodeColor=<rgba 1 0 0.0469 1>}
  MaterialExpressionReroute_15 = Reroute(None=MaterialFunctionCall(
    Color_Morning=VectorParameter() {parameterName=Color.Phase1.Morning, defaultValue=<rgba 1 1 1 1>}.RGB
    Color_Day=VectorParameter() {parameterName=Color.Phase1.Day, defaultValue=<rgba 1 1 1 1>}.RGB
    Color_Evening=VectorParameter() {parameterName=Color.Phase1.Evening, defaultValue=<rgba 1 1 1 1>}.RGB
    Color_Night=VectorParameter() {parameterName=Color.Phase1.Night, defaultValue=<rgba 1 1 1 1>}.RGB
   ) {materialFunction=&Color_Time_of_Day}.Result)
  MaterialExpressionMaterialFunctionCall_12 = MaterialFunctionCall(
   Color_Morning=VectorParameter() {parameterName=Color.Phase2.Morning, defaultValue=<rgba 1 1 1 1>}.RGB
   Color_Day=VectorParameter() {parameterName=Color.Phase2.Day, defaultValue=<rgba 1 1 1 1>}.RGB
   Color_Evening=VectorParameter() {parameterName=Color.Phase2.Evening, defaultValue=<rgba 1 1 1 1>}.RGB
   Color_Night=VectorParameter() {parameterName=Color.Phase2.Night, defaultValue=<rgba 1 1 1 1>}.RGB
  ) {materialFunction=&Color_Time_of_Day}
  n70 = ScalarParameter() {defaultValue=1, parameterName=SourceRadius.Beat.MinMulti}
  n71 = ScalarParameter() {defaultValue=1.05, parameterName=SourceRadius.Beat.MaxMulti}
  n72 = MaterialFunctionCall!(In) {materialFunction=/ClydeCosmetics/Environments/KelpLinen/MF_Kelp_AudioReactive}
  n73 = SphereMask(
   A=WorldPosition().XYZ
   B=VectorParameter() {group=Source Light, parameterName=SourcePoint, defaultValue=<rgba 0 0 0 1>}.RGB
   Radius=Multiply(
    A=ScalarParameter() {defaultValue=5000, group=Source Light, parameterName=SourceRadius}
    B=LinearInterpolate(A=@n70, B=@n71, Alpha=@n72.Beat_Tracker)
   )
   Hardness=ScalarParameter() {defaultValue=0.5, group=Source Light, parameterName=SourceFalloff}
  )
  MaterialExpressionLinearInterpolate_4 = LinearInterpolate(
   A=MaterialFunctionCall(
    Color_Morning=VectorParameter() {parameterName=Color.Unlit.Morning, defaultValue=<rgba 1 1 1 1>}.RGB
    Color_Day=VectorParameter() {parameterName=Color.Unlit.Day, defaultValue=<rgba 1 1 1 1>}.RGB
    Color_Evening=VectorParameter() {parameterName=Color.Unlit.Evening, defaultValue=<rgba 1 1 1 1>}.RGB
    Color_Night=VectorParameter() {parameterName=Color.Unlit.Night, defaultValue=<rgba 1 1 1 1>}.RGB
   ) {materialFunction=&Color_Time_of_Day}.Result
   B=MaterialFunctionCall(
    Color_Morning=VectorParameter() {parameterName=Color.Glow.Morning, defaultValue=<rgba 1 1 1 1>}.RGB
    Color_Day=VectorParameter() {parameterName=Color.Glow.Day, defaultValue=<rgba 1 1 1 1>}.RGB
    Color_Evening=VectorParameter() {parameterName=Color.Glow.Evening, defaultValue=<rgba 1 1 1 1>}.RGB
    Color_Night=VectorParameter() {parameterName=Color.Glow.Night, defaultValue=<rgba 1 1 1 1>}.RGB
   ) {materialFunction=&Color_Time_of_Day}.Result
   Alpha=@n73
  )
  n74 = Switch(
   SwitchValue=CollectionParameter() {collection=&MPC_KelpLinen, parameterName=Phase_Tracker}
   Default=@MaterialExpressionReroute_15
   0=@MaterialExpressionReroute_15
   1=@MaterialExpressionReroute_15
   2=@MaterialExpressionMaterialFunctionCall_12.Result
   3=@MaterialExpressionLinearInterpolate_4
  ) {inputs [
    {
     inputName=0
     input {
      expression=&self:MaterialExpressionReroute_15
      inputName=None
      mask=0
      maskA=0
      maskB=0
      maskG=0
      maskR=0
      outputIndex=0
     }
    }
    {
     inputName=1
     input {
      expression=&self:MaterialExpressionReroute_15
      inputName=None
      mask=0
      maskA=0
      maskB=0
      maskG=0
      maskR=0
      outputIndex=0
     }
    }
    {
     inputName=2
     input {
      expression=&self:MaterialExpressionMaterialFunctionCall_12
      inputName=None
      mask=0
      maskA=0
      maskB=0
      maskG=0
      maskR=0
      outputIndex=0
     }
    }
    {
     inputName=3
     input {
      expression=&self:MaterialExpressionLinearInterpolate_4
      inputName=None
      mask=0
      maskA=0
      maskB=0
      maskG=0
      maskR=0
      outputIndex=0
     }
    }
   ]}
  n75 = ScalarParameter() {parameterName=Color.SecondaryNoise.Blend}
  n76 = LinearInterpolate(
   A=Multiply(A=VectorParameter() {parameterName=Color.Alt, defaultValue=<rgba 1 1 1 1>}.RGB, B=@n74)
   B=@n74
   Alpha=LinearInterpolate(B=NamedRerouteUsage(), Alpha=Saturate(None=@n75)) {constA=1}
  )
  Reroute(None=@n76)
  n77 = LinearInterpolate(
   B=LinearInterpolate(
    A=ScalarParameter() {parameterName=Emissive.Multi.Beat.Min}
    B=ScalarParameter() {defaultValue=1, parameterName=Emissive.Multi.Beat.Max}
    Alpha=MaterialFunctionCall() {materialFunction=&MF_Kelp_AudioReactive}.Beat_Tracker
   )
   Alpha=@n73
  ) {constA=1}
  n78 = Multiply(
   A=@n76
   B=MaterialFunctionCall(
    Value_Morning=ScalarParameter() {defaultValue=1, parameterName=Emissive.Multi.Morning}
    Value_Day=ScalarParameter() {defaultValue=1, parameterName=Emissive.Multi.Day}
    Value_Evening=ScalarParameter() {defaultValue=1, parameterName=Emissive.Multi.Evening}
    Value_Night=ScalarParameter() {defaultValue=1, parameterName=Emissive.Multi.Night}
   ) {materialFunction=&Scalar_Time_of_Day}.Result
  )
  Reroute(None=Multiply(A=Multiply(A=@n77, B=@n78), B=ScalarParameter() {defaultValue=1, parameterName=Emissive.Multi.Master}))
  ScalarParameter() {defaultValue=0.2, parameterName=Emissive.NoiseLow.Multi}
  n79 = MaterialFunctionCall(Float2=StaticSwitchParameter(True=TextureCoordinate() {coordinateIndex=1}, False=TextureCoordinate()) {defaultValue=true, parameterName=WPO.UVChannel2}) {materialFunction=&BreakOutFloat2Components}
  n80 = Subtract(A=AppendVector(A=@n79.R, B=OneMinus(None=@n79.G))) {constB=0.5}
  n81 = DotProduct(A=@n80, B=@n80)
  n82 = Divide(A=@n80, B=Max(A=SquareRoot(None=@n81)) {constB=0})
  n83 = Distance(A=WorldPosition().XYZ, B=ObjectPositionWS())
  n84 = Transform(None=AppendVector(A=Multiply(A=@n82, B=@n83), B=Constant())) {transformSourceType=TRANSFORMSOURCE_Camera}
  n85 = ObjectPositionWS()
  StaticSwitchParameter(True=Subtract(A=Add(A=@n84, B=@n85), B=WorldPosition().XYZ), False=Constant()) {defaultValue=true, parameterName=WPO.UseCameraFacing}
 ]
}
